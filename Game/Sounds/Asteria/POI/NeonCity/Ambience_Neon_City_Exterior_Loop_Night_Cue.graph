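# Looping nighttime exterior ambience cue for the Neon City POI under Asteria.
# A looping Wave Player plays a short quad night wave into an Enveloper, whose volume curve ramps from silence to full as a fade-in.
# The cue plays at reduced volume, using Athena ambient attenuation and the outdoor ambient sound class.
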SoundCue /Game/Sounds/Asteria/POI/NeonCity/Ambience_Neon_City_Exterior_Loop_Night_Cue
root=Enveloper_4
nodes [
 {
  id=Enveloper_4
  children [WavePlayer_4]
  properties {pitchCurve {editorCurveData {keys [{time=0}]}}, volumeCurve {editorCurveData {keys [<at 0 0>,<at 1 1>]}}}
 }
 {id=WavePlayer_4, properties {bLooping=true, soundWaveAssetPtr=&/Ambience_Neon_City_Ext_Night_Quad_B_Short}}
]
properties {
 attenuationSettings=&Fort_Ambient_Athena_Att
 maxDistance=2097152
 soundClassObject=&Fort_Ambient_Outdoors_Athena
 volumeMultiplier=0.501
 preEffectBusSends [{}]
}
