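# Sound cue for the exterior rain ambience at the Neon City POI in Asteria.
# A looping wave player feeds the quad rain bed into a continuous modulator, which drives volume from the Amb.Height parameter.
# The cue uses the Athena ambient attenuation and outdoor ambient sound class, with a slightly reduced overall volume.
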SoundCue /Game/Sounds/Asteria/POI/NeonCity/Ambience_Neon_City_Rain_Exterior_Cue
root=ModulatorContinuous_1
nodes [
 {id=ModulatorContinuous_1, children [WavePlayer_8], properties {volumeModulationParams {default=0, parameterName=Amb.Height}}}
 {
  id=WavePlayer_8
  properties {bLooping=true, soundWaveAssetPtr=&/Ambience_Neon_City_Ext_Rain_Quad_Default_Loop}
 }
]
properties {
 attenuationSettings=&Fort_Ambient_Athena_Att
 maxDistance=2097152
 soundClassObject=&Fort_Ambient_Outdoors_Athena
 volumeMultiplier=0.827
 preEffectBusSends [{}]
}
