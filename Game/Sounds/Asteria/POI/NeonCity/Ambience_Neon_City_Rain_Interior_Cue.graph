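# Sound cue for light interior rain ambience in the Asteria Neon City POI, using the indoor ambient sound class and ambient attenuation.
# A quality-level root routes its first two slots to a random pick of three quad waves and its last two to a random pick of two stereo waves.
# Each branch's enveloper loops a gentle volume dip indefinitely, while only the stereo wave players are set to loop.
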SoundCue /Game/Sounds/Asteria/POI/NeonCity/Ambience_Neon_City_Rain_Interior_Cue
root=QualityLevel_1
nodes [
 {
  id=Enveloper_0
  children [Random_0]
  properties {
   bLoop=true
   bLoopIndefinitely=true
   loopEnd=20
   pitchCurve {editorCurveData {keys [{time=0}]}}
   volumeCurve {editorCurveData {keys [
      {arriveTangent=-0.01, interpMode=RCIM_Cubic, leaveTangent=-0.01, time=0, value=1}
      <at 10 0.9 RCIM_Cubic>
      {arriveTangent=0.01, interpMode=RCIM_Cubic, leaveTangent=0.01, time=20, value=1}
     ]}}
  }
 }
 {
  id=Enveloper_3
  children [Random_1]
  properties {
   bLoop=true
   bLoopIndefinitely=true
   loopEnd=20
   pitchCurve {editorCurveData {keys [{time=0}]}}
   volumeCurve {editorCurveData {keys [
      {arriveTangent=-0.01, interpMode=RCIM_Cubic, leaveTangent=-0.01, time=0, value=1}
      <at 10 0.9 RCIM_Cubic>
      {arriveTangent=0.01, interpMode=RCIM_Cubic, leaveTangent=0.01, time=20, value=1}
     ]}}
  }
 }
 {id=QualityLevel_1, children [Enveloper_3,Enveloper_3,Enveloper_0,Enveloper_0]}
 {id=Random_0, children [WavePlayer_5,WavePlayer_8], properties {weights [1,1]}}
 {id=Random_1, children [WavePlayer_0,WavePlayer_1,WavePlayer_6], properties {weights [1,1,1]}}
 {id=WavePlayer_0, properties {soundWaveAssetPtr=&/Ambience_Neon_City_Int_Rain_Interior_Light_01_Quad}}
 {id=WavePlayer_1, properties {soundWaveAssetPtr=&/Ambience_Neon_City_Int_Rain_Interior_Light_02_Quad}}
 {
  id=WavePlayer_5
  properties {bLooping=true, soundWaveAssetPtr=&/Ambience_Neon_City_Int_Rain_Interior_Light_01_Stereo}
 }
 {id=WavePlayer_6, properties {soundWaveAssetPtr=&/Ambience_Neon_City_Int_Rain_Interior_Light_03_Quad}}
 {
  id=WavePlayer_8
  properties {bLooping=true, soundWaveAssetPtr=&/Ambience_Neon_City_Int_Rain_Interior_Light_02_Stereo}
 }
]
properties {
 attenuationSettings=&Fort_Ambient_Athena_Att
 maxDistance=2097152
 soundClassObject=&Fort_Ambient_Indoors_Athena
 volumeMultiplier=0.601
 preEffectBusSends [{}]
}
